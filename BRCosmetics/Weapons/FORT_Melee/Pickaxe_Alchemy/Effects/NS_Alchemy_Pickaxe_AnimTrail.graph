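# Niagara ribbon trail for the Alchemy pickaxe, built on the shared AnimTrail modules.
# A single Trail emitter spawns, initializes and updates particles through those modules, using bone or socket indices, with a skeletal mesh exposed as a user parameter.
# The short-lived ribbon takes a faint blue color selected by time of day, with only the night tint differing slightly.
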NiagaraSystem /BRCosmetics/Weapons/FORT_Melee/Pickaxe_Alchemy/Effects/NS_Alchemy_Pickaxe_AnimTrail
dependencies {usedDataInterfaces [], usedDynamicInputs [], usedModules [&SystemState]}
emitters {Trail {
  emitterUpdateScript [
   &EmitterState()
   &FN_SpawnAnimTrail_v2()
  ]
  particleSpawnScript [
   /Game/Effects/Niagara/Modules/AnimTrail/FN_InitializeAnimTrail/FN_InitializeAnimTrailBoneIndecies(First Bone or Socket Index=1)
   /Game/Effects/Niagara/Modules/AnimTrail/FN_InitializeAnimTrail/FN_InitializeAnimTrail(
    AnimTrailWidth=1
    Color=&FN_Select_Color_from_Time_Of_Day(Morning=<rgba 0.0802 0.6168 1 0.1>, Day=<rgba 0.0802 0.6168 1 0.1>, Evening=<rgba 0.0802 0.6168 1 0.1>, Night=<rgba 0.0802 0.8796 1 0.1>)
    Lifetime=0.2
   )
  ]
  particleUpdateScript [&FN_UpdateAnimTrail_v2()]
  rendererClasses [&NiagaraRibbonRendererProperties]
  renderers [{rendererClass=&NiagaraRibbonRendererProperties}]
 }}
summary {
 systemName=NS_Alchemy_Pickaxe_AnimTrail
 emitters [{emitterName=Trail, rendererClasses [&NiagaraRibbonRendererProperties]}]
 userVariables [
  User.LifetimeCustom {type=NiagaraFloat, defaultValue=0, description=""}
  User.RibbonIDOffset {type=NiagaraInt32}
  User.SkeletalMesh {
   type=NiagaraDataInterfaceSkeletalMesh
   defaultValue {
    struct=NiagaraExt_VariableValue_DataInterface
    value {dataInterface=&self:NiagaraDataInterfaceSkeletalMesh_0}
   }
  }
 ]
}
