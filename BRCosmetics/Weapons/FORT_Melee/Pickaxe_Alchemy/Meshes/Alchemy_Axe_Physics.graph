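# Physics asset for the Alchemy pickaxe that defines bodies and joints for the dynamic feathers on its head and handle.
# Four chains run outward from small sphere base bodies through box segments: left and right feathers, plus top and bottom handle feathers.
# Joints have tight swing and twist limits, except a looser twist at the left feather root.
# Two unconstrained master bodies supply larger shapes: a handle capsule and a multi-shape feather compound.
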PhysicsAsset /BRCosmetics/Weapons/FORT_Melee/Pickaxe_Alchemy/Meshes/Alchemy_Axe_Physics
bodies {
 dyn_feather_01_l [{shapeType=Box, extentX=5.3467, extentY=1.1468, extentZ=14.0726, center=<7.2711 -0.1974 0.1695>, rotation=<rot 0 91.8092 -85.3828>}]
 dyn_feather_01_r [{shapeType=Box, extentX=4.4321, extentY=1.1254, extentZ=9.4822, center=<5.1516 -0.1607 0.0496>, rotation=<rot 0 90.7207 -89.5249>}]
 dyn_feather_02_l [{shapeType=Box, extentX=5.1475, extentY=1.7468, extentZ=9.3732, center=<4.394 -0.154 0.0045>, rotation=<rot 0 86.5013 -89.803>}]
 dyn_feather_base_l [{shapeType=Sphere, radius=0.5}]
 dyn_feather_base_r [{shapeType=Sphere, radius=0.5}]
 dyn_feather_handle_01_bottom [{shapeType=Box, extentX=11.537, extentY=4.0144, extentZ=1.5705, center=<4.6077 0.0001 0>}]
 dyn_feather_handle_01_top [{shapeType=Box, extentX=13.1645, extentY=3.2146, extentZ=1.6129, center=<6.2481 -0.0001 0>}]
 dyn_feather_handle_base_bottom [{shapeType=Sphere, radius=0.5}]
 dyn_feather_handle_base_top [{shapeType=Sphere, radius=0.5}]
 dyn_feather_handle_master [{shapeType=Capsule, length=15.7928, radius=4.6008, center=<20.7542 -0.5668 0>, rotation=<rot 90 0 0>}]
 dyn_feather_handle_parent_bottom [{shapeType=Box, extentX=13.1192, extentY=3.5259, extentZ=1.8278, center=<6.9329 0 0>}]
 dyn_feather_handle_parent_top [{shapeType=Box, extentX=9.7998, extentY=2.4218, extentZ=1.6377, center=<5.0456 0 0>}]
 dyn_feather_parent_l [{shapeType=Box, extentX=4.0034, extentY=1.0787, extentZ=3.8688, center=<2.052 -0.0199 0>, rotation=<rot 0 89.8418 -89.9539>}]
 dyn_feather_parent_r [{shapeType=Box, extentX=4.0034, extentY=1.2926, extentZ=3.8688, center=<1.8223 -0.0068 -0.018>, rotation=<rot 0 -89.8571 -89.4401>}]
 dyn_feathers_master [
  {shapeType=Capsule, length=27.1984, radius=2.8016, center=<13.1419 0 -0.0001>, rotation=<rot 90 0 0>}
  {shapeType=Capsule, length=9.6782, radius=2.8944, center=<-3.4499 -3.1187 4.9956>, rotation=<rot -56.0917 0 -14.0183>}
  {shapeType=Capsule, length=9.6782, radius=2.8944, center=<-3.6368 3.1178 4.8757>, rotation=<rot -56.0911 0 14.0979>}
  {shapeType=Capsule, length=3.923, radius=3.8471, center=<6.0399 0 -0.0001>, rotation=<rot 90 0 0>}
  {shapeType=Box, extentX=26.4435, extentY=27.9029, extentZ=2.8765, center=<-0.1044 4.5253 3.955>}
 ]
 dyn_reather_02_r [{shapeType=Box, extentX=4.4864, extentY=1.1215, extentZ=9.1256, center=<4.4504 -0.2697 0.0088>, rotation=<rot 0 85.1281 -89.7904>}]
}
constraints [
 {bone1Name=dyn_feather_parent_l, bone2Name=dyn_feather_base_l, swing1LimitDegrees=5, swing2LimitDegrees=5, twistLimitDegrees=50}
 {bone1Name=dyn_feather_01_l, bone2Name=dyn_feather_parent_l, swing1LimitDegrees=5, swing2LimitDegrees=5, twistLimitDegrees=3}
 {bone1Name=dyn_feather_02_l, bone2Name=dyn_feather_01_l, swing1LimitDegrees=2, swing2LimitDegrees=2, twistLimitDegrees=1}
 {bone1Name=dyn_reather_02_r, bone2Name=dyn_feather_01_r, swing1LimitDegrees=5, swing2LimitDegrees=5, twistLimitDegrees=3}
 {bone1Name=dyn_feather_01_r, bone2Name=dyn_feather_parent_r, swing1LimitDegrees=10, swing2LimitDegrees=10, twistLimitDegrees=3}
 {bone1Name=dyn_feather_parent_r, bone2Name=dyn_feather_base_r, swing1LimitDegrees=8, swing2LimitDegrees=8, twistLimitDegrees=3}
 {
  bone1Name=dyn_feather_handle_parent_bottom
  bone2Name=dyn_feather_handle_base_bottom
  swing1LimitDegrees=10
  swing2LimitDegrees=10
  twistLimitDegrees=3
 }
 {
  bone1Name=dyn_feather_handle_01_bottom
  bone2Name=dyn_feather_handle_parent_bottom
  swing1LimitDegrees=5
  swing2LimitDegrees=5
  twistLimitDegrees=3
 }
 {
  bone1Name=dyn_feather_handle_01_top
  bone2Name=dyn_feather_handle_parent_top
  swing1LimitDegrees=5
  swing2LimitDegrees=5
  twistLimitDegrees=3
 }
 {
  bone1Name=dyn_feather_handle_parent_top
  bone2Name=dyn_feather_handle_base_top
  swing1LimitDegrees=10
  swing2LimitDegrees=10
  twistLimitDegrees=3
 }
]
